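# Placeholder material for the Veneer tech-art tool, a Default Lit surface with an animated, view-dependent grayscale gradient.
# A linear Fresnel term minus Time passes through Frac, producing a repeating ramp that sweeps toward silhouette edges once per second.
# The result sets only Base Color through SetMaterialAttributes, with no textures or parameters.
# Automatic usage-flag setting in the editor is turned off.
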Material /Game/ArtTools/Tech_Art/Veneer/Misc/Materials/M_VeneerPlaceholder
graph {
 properties {bAutomaticallySetUsageInEditor=false, bUseMaterialAttributes=true, usedShadingModels=Default Lit}
 nodes [
  n1 = Fresnel() {baseReflectFraction=0, exponent=1}
  n2 = Time()
  SetMaterialAttributes!(MaterialAttributes, Base Color=Frac(None=Subtract(A=@n1, B=@n2))) {attributeSetTypes [#1]}
 ]
}
